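# NeonPunk wheel material for vehicle cosmetics that animates part of the mesh by oscillating rotation in World Position Offset.
# MF_WheelParent_Simple supplies base attributes; its WPO blends by vertex-color blue toward a cosine-timed RotateAboutAxis.
# The axis derives from Axis_Rotation and object orientation, while the WPO_Mesh_is_SK static switch selects hardcoded pivot and position offsets.
# The modified attributes then pass through MF_BaseVehicle_Damage and MF_VehicleDeRez_Transition.
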Material /VehicleCosmetics/Wheels/NeonPunk/Materials/M_Wheel_NeonPunk
graph {
 properties {
  bAutomaticallySetUsageInEditor=false
  bUseMaterialAttributes=true
  bUsedWithMeshDeformer=true
  bUsedWithSkeletalMesh=true
  pixelDepthOffsetMode=PDOM_Legacy
  usedShadingModels=Default Lit
 }
 nodes [
  Constant4Vector() {constant=<rgba 3.85 0 -3.75 1>}
  Constant4Vector() {constant=<rgba -3.85 0 -3.5 1>}
  n1 = GetMaterialAttributes(None=MaterialFunctionCall() {materialFunction=&MF_WheelParent_Simple}.MaterialAttributes) {attributeGetTypes [#1]}
  n2 = ComponentMask(None=VectorParameter() {parameterName=Axis_Rotation, defaultValue=<rgba 0 1 0 1>}.RGBA) {b=true, g=true, r=true}
  n3 = Normalize(VectorInput=Transform(None=@n2) {transformSourceType=TRANSFORMSOURCE_Local})
  n4 = Transform(None=@n2) {transformSourceType=TRANSFORMSOURCE_Local}
  n5 = CrossProduct(A=@n4, B=ObjectOrientation())
  n6 = StaticComponentMaskParameter(None=Normalize(VectorInput=Transform(None=@n5) {transformSourceType=TRANSFORMSOURCE_World, transformType=TRANSFORM_Local})) {defaultR=true, desc=WPO, parameterName=MaskParametr}
  n7 = Multiply(A=ScalarParameter() {defaultValue=1, parameterName=TimeScale}, B=Time())
  n8 = Add(A=Multiply(A=Cosine(None=@n7), B=Constant() {r=0.5}), B=Constant() {r=0.5})
  n9 = Constant() {r=-0.5}
  n10 = ObjectPositionWS()
  n11 = Constant4Vector() {constant=<rgba -12.12 0 11.912 1>}
  n12 = Constant4Vector() {constant=<rgba 0 0 0 1>}
  n13 = ComponentMask(None=Transform(None=StaticSwitchParameter(True=@n11.RGB, False=@n12.RGB) {defaultValue=true, parameterName=WPO_Mesh_is_SK}) {transformSourceType=TRANSFORMSOURCE_Local}) {b=true, g=true, r=true}
  n14 = WorldPosition()
  n15 = Constant4Vector() {constant=<rgba -12.12 0 11.912 1>}
  n16 = ComponentMask(None=Transform(None=StaticSwitchParameter(True=@n12.RGBA, False=@n15.RGB) {defaultValue=true, parameterName=WPO_Mesh_is_SK}) {transformSourceType=TRANSFORMSOURCE_Local}) {b=true, g=true, r=true}
  n17 = LinearInterpolate(
   A=@n1.World Position Offset
   B=RotateAboutAxis(
    NormalizedRotationAxis=Multiply(A=@n3, B=@n6)
    RotationAngle=Multiply(A=@n8, B=@n9)
    PivotPoint=Add(A=@n10, B=@n13)
    Position=Add(A=@n14.XYZ, B=@n16)
   )
   Alpha=VertexColor().B
  )
  MaterialFunctionCall(In=MaterialFunctionCall(Material=SetMaterialAttributes!(MaterialAttributes=@n1.MaterialAttributes, World Position Offset=@n17) {attributeSetTypes [#1]}) {materialFunction=&MF_BaseVehicle_Damage}.Result) {materialFunction=&MF_VehicleDeRez_Transition}
  MaterialFunctionCall() {materialFunction=&ObjectPivotPoint}
  Reroute(None=@n4)
 ]
}
