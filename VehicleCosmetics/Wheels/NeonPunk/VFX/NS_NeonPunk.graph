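# Niagara effect for the NeonPunk vehicle wheel cosmetic that puffs steam along +Y from three offset points: top, bottom left and bottom right.
# Steam emitters pulse sprites via a spawn-rate curve over normalized loop age and generate location events.
# Paired Cloud emitters have zero or disabled spawning, so presumably rely on those events, though no handlers appear here.
# Curves shape alpha, size and most velocities, while dynamic material parameters feed the sprite materials.
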NiagaraSystem /VehicleCosmetics/Wheels/NeonPunk/VFX/NS_NeonPunk
dependencies {
 usedDataInterfaces [
  &NiagaraDataInterfaceCurve
  &NiagaraDataInterfaceVectorCurve
  &NiagaraDataInterfaceVector2DCurve
 ]
 usedDynamicInputs []
 usedModules [&SystemState]
}
emitters {
 CloudBottomLeft {
  emitterUpdateScript [
   &EmitterState()
   &SpawnBurst_Instantaneous(Spawn Count=0, Spawn Time=0.1, Use Spawn Probability=false, UseLoopCountLimit=false) {enabled=false}
   &SpawnRate(SpawnRate=0, Use Spawn Probability=false)
  ]
  particleSpawnScript [
   &InitializeParticle(
    Color=<rgba 1 1 1 0.45>
    Lifetime Max=1.15
    Lifetime Min=0.8
    Lifetime Mode=::NewEnumerator1
    Position=<-24 18 -16>
    Position Mode=::NewEnumerator1
    Sprite Rotation Mode=::NewEnumerator2
    Sprite Size Mode=::NewEnumerator1
    Uniform Sprite Size Max=50
    Uniform Sprite Size Min=40
    UsePositionOffset=false
    Write Material Random=true
   )
   &AddVelocity(Velocity=&RandomRangeVector(Minimum=<-5 50 -5>, Maximum=<5 85 5>))
   &DynamicMaterialParameters(
    Index 0 Param 1=0.05
    Index 0 Param 2=0.5
    Index 0 Param 3=2
    Module.Index0Float0Write=true
    Module.Index0Float1Write=true
    Module.Index0Float2Write=true
    Module.Index0Float3Write=true
   )
  ]
  particleUpdateScript [
   &ParticleState()
   &ScaleColor(
    Scale Alpha=&FloatFromCurve(FloatCurve=DataInterface() {propertyValues {Curve {keys [<at 0 0>,<at 0.05 1>,<at 0.1 0.2>,<at 1 0>]}}})
    ScaleA=true
    ScaleRGB=true
   )
   &ScaleVelocity(Velocity Scale=&VectorFromCurve(VectorCurve=DataInterface() {propertyValues {
       XCurve {keys [<at 0 0>,<at 0.15 0>,<at 0.8 0.2>]}
       YCurve {keys [<at 0 0.25>,<at 0.15 0.2>,<at 0.8 0.15>]}
       ZCurve {keys [<at 0 0.1313>,<at 0.15 0>,<at 0.8 0.2>]}
      }}))
   &SolveForcesAndVelocity(Clamp Velocity=false, Limit Acceleration=false)
   &ScaleSpriteSize(Uniform Curve Sprite Scale=DataInterface() {propertyValues {Curve {keys [<at 0.3 0.1>,<at 0.7 2>]}}})
  ]
  rendererClasses [&NiagaraSpriteRendererProperties]
  renderers [{rendererClass=&NiagaraSpriteRendererProperties}]
 }
 CloudBottomRight {
  emitterUpdateScript [
   &EmitterState()
   &SpawnBurst_Instantaneous(Spawn Count=0, Spawn Time=0.1, Use Spawn Probability=false, UseLoopCountLimit=false) {enabled=false}
   &SpawnRate(SpawnRate=10, Use Spawn Probability=false) {enabled=false}
  ]
  particleSpawnScript [
   &InitializeParticle(
    Color=<rgba 1 1 1 0.8>
    Lifetime Max=1.3
    Lifetime Min=0.8
    Lifetime Mode=::NewEnumerator1
    Position=<10 18 -16>
    Position Mode=::NewEnumerator1
    Sprite Rotation Mode=::NewEnumerator2
    Sprite Size Mode=::NewEnumerator1
    Uniform Sprite Size Max=30
    Uniform Sprite Size Min=20
    UsePositionOffset=false
    Write Material Random=true
   )
   &AddVelocity(Velocity=&RandomRangeVector(Minimum=<-5 45 -5>, Maximum=<5 50 5>))
   &DynamicMaterialParameters(
    Index 0 Param 1=0.05
    Index 0 Param 2=0.5
    Index 0 Param 3=2
    Module.Index0Float0Write=true
    Module.Index0Float1Write=true
    Module.Index0Float2Write=true
    Module.Index0Float3Write=true
   )
   &AddRotationalVelocity(Rotation Rate=&RandomRangeVector(Minimum=<-0.02 -0.02 -0.02>, Maximum=<0.02 0.02 0.02>))
  ]
  particleUpdateScript [
   &ParticleState()
   &ScaleColor(
    Scale Alpha=&FloatFromCurve(FloatCurve=DataInterface() {propertyValues {Curve {keys [<at 0 0>,<at 0.1469 0.7617>,<at 0.3996 0.4975>,<at 1 0>]}}})
    ScaleA=true
    ScaleRGB=true
   )
   &ScaleVelocity(Velocity Scale=&VectorFromCurve(VectorCurve=DataInterface() {propertyValues {
       XCurve {keys [<at 0 0>,<at 0.15 0>,<at 0.8 0.2>]}
       YCurve {keys [<at 0 0.25>,<at 0.15 0.2>,<at 0.8 0.15>]}
       ZCurve {keys [<at 0 0>,<at 0.15 0>,<at 0.8 0.2>]}
      }}))
   &SolveForcesAndVelocity(Clamp Velocity=false, Limit Acceleration=false)
   &ScaleSpriteSize(Uniform Curve Sprite Scale=DataInterface() {propertyValues {Curve {keys [<at 0 0.4>,<at 0.3 0.8>,<at 0.7 1.2>,<at 1 1.6>]}}})
  ]
  rendererClasses [&NiagaraSpriteRendererProperties]
  renderers [{rendererClass=&NiagaraSpriteRendererProperties}]
 }
 CloudTop {
  emitterUpdateScript [
   &EmitterState()
   &SpawnBurst_Instantaneous(Spawn Count=0, Spawn Time=0.1, Use Spawn Probability=false, UseLoopCountLimit=false) {enabled=false}
   &SpawnRate(SpawnRate=0, Use Spawn Probability=false)
  ]
  particleSpawnScript [
   &InitializeParticle(
    Color=<rgba 1 1 1 0.45>
    Lifetime Max=1.15
    Lifetime Min=0.8
    Lifetime Mode=::NewEnumerator1
    Position=<18 15 20>
    Position Mode=::NewEnumerator1
    Position Offset=<0 20 0>
    Sprite Rotation Mode=::NewEnumerator2
    Sprite Size Mode=::NewEnumerator1
    Uniform Sprite Size Max=80
    Uniform Sprite Size Min=60
    UsePositionOffset=false
    Write Material Random=true
   )
   &AddVelocity(Velocity=&RandomRangeVector(Minimum=<-5 85 -5>, Maximum=<5 50 5>))
   &DynamicMaterialParameters(
    Index 0 Param 1=0.5
    Index 0 Param 2=1
    Index 0 Param 3=2
    Module.Index0Float0Write=true
    Module.Index0Float1Write=true
    Module.Index0Float2Write=true
    Module.Index0Float3Write=true
   )
  ]
  particleUpdateScript [
   &ParticleState()
   &ScaleColor(
    Scale Alpha=&FloatFromCurve(FloatCurve=DataInterface() {propertyValues {Curve {keys [<at 0 0>,<at 0.3 1.5>,<at 0.36 1>,<at 0.4 0.5>,<at 1 0>]}}})
    ScaleA=true
    ScaleRGB=true
   )
   &ScaleVelocity(Velocity Scale=&VectorFromCurve(VectorCurve=DataInterface() {propertyValues {
       XCurve {keys [<at 0 0>,<at 0.15 0>,<at 1 0.2>]}
       YCurve {keys [<at 0 0.25>,<at 0.15 0.2>,<at 0.8 0.15>]}
       ZCurve {keys [<at 0 0.25>,<at 0.8 0.2>]}
      }}))
   &SolveForcesAndVelocity(Clamp Velocity=false, Limit Acceleration=false)
   &ScaleSpriteSize(Uniform Curve Sprite Scale=DataInterface() {propertyValues {Curve {keys [<at 0 0.25>, {tangentMode=RCTM_User, time=0.12, value=0.6}, <at 0.35 1.5>, <at 0.6248 2>, <at 1 0>]}}})
  ]
  rendererClasses [&NiagaraSpriteRendererProperties]
  renderers [{rendererClass=&NiagaraSpriteRendererProperties}]
 }
 SteamBottomLeft {
  emitterUpdateScript [
   &EmitterState()
   &SpawnBurst_Instantaneous(Spawn Count=40, Spawn Time=0.1, Use Spawn Probability=false, UseLoopCountLimit=false) {enabled=false}
   &SpawnRate(
    SpawnRate=&FloatFromCurve(
     FloatCurve=DataInterface() {propertyValues {Curve {keys [<at 0 36>,<at 0.12 36>,<at 0.13 0.1>,<at 0.75 0.1>,<at 0.76 36>,<at 1 36>]}}}
     CurveIndex=$Emitter.NormalizedLoopAge
    )
    Use Spawn Probability=false
   )
  ]
  particleSpawnScript [
   &InitializeParticle(
    Color=<rgba 1 1 1 0.45>
    Lifetime Max=1.15
    Lifetime Min=0.8
    Lifetime Mode=::NewEnumerator1
    Position=<-24 18 -16>
    Position Mode=::NewEnumerator1
    Sprite Rotation Mode=::NewEnumerator2
    Sprite Size Mode=::NewEnumerator1
    Uniform Sprite Size Max=20
    Uniform Sprite Size Min=10
    UsePositionOffset=false
    Write Material Random=true
   )
   &OffsetPosition(Position Offset=<7.5 2.5 -4.5>)
   &AddVelocity(Velocity=&RandomRangeVector(Minimum=<0 50 0>, Maximum=<0 85 5>))
   &AddRotationalVelocity(Rotation Rate=&RandomRangeVector(Minimum=<-0.02 -0.02 -0.02>, Maximum=<0.02 0.02 0.02>))
   &DynamicMaterialParameters(
    Index 0 Param 1=0.05
    Index 0 Param 2=0.5
    Index 0 Param 3=2
    Module.Index0Float0Write=true
    Module.Index0Float1Write=true
    Module.Index0Float2Write=true
    Module.Index0Float3Write=true
   )
  ]
  particleUpdateScript [
   &ParticleState()
   &ScaleColor(
    Scale Alpha=&FloatFromCurve(FloatCurve=DataInterface() {propertyValues {Curve {keys [<at 0 0.2>,<at 0.05 0.6>,<at 0.3 0.3>,<at 1 0>]}}})
    ScaleA=true
    ScaleRGB=true
   )
   &ScaleVelocity(Velocity Scale=&VectorFromCurve(VectorCurve=DataInterface() {propertyValues {
       XCurve {keys [<at 0 0.2 RCIM_Cubic>,<at 0.15 0 RCIM_Cubic>,<at 0.8 0.2>,<at 1 0.2>]}
       YCurve {keys [
         <at 0.0038 0 RCIM_Cubic>
         {arriveTangent=0.9537, interpMode=RCIM_Cubic, leaveTangent=0.9537, time=0.15, value=0.8}
         <at 0.1857 0.1735>
         <at 0.8 0.15>
         <at 1 0.15>
        ]}
       ZCurve {keys [<at 0 0.25 RCIM_Cubic>, {arriveTangent=-0.0625, interpMode=RCIM_Cubic, leaveTangent=-0.0625, time=0.15, value=0}, <at 0.8 0.2>, <at 1 0.2>]}
      }}))
   &SolveForcesAndVelocity(Clamp Velocity=false, Limit Acceleration=false)
   &GenerateLocationEvent(Delay Before Sending Events=0.45, Event Send Rate=2, Use Event Probability=false, UseEventDelay=true)
   &ScaleSpriteSize(
    Non-Uniform Curve Sprite Scale=DataInterface() {propertyValues {
      XCurve {keys [{tangentMode=RCTM_User, time=0, value=0.1}, <at 0.3 0.1>, {tangentMode=RCTM_User, time=0.7, value=0.7}]}
      YCurve {keys [{tangentMode=RCTM_User, time=0, value=0.1}, <at 0.2916 0.5>, {tangentMode=RCTM_User, time=0.7, value=0.8}]}
     }}
    Scale Sprite Size Mode=::NewEnumerator3
   )
  ]
  rendererClasses [&NiagaraSpriteRendererProperties]
  renderers [{rendererClass=&NiagaraSpriteRendererProperties}]
 }
 SteamBottomRight {
  emitterUpdateScript [
   &EmitterState()
   &SpawnBurst_Instantaneous(Spawn Count=40, Spawn Time=0.1, Use Spawn Probability=false, UseLoopCountLimit=false) {enabled=false}
   &SpawnRate(
    SpawnRate=&FloatFromCurve(
     FloatCurve=DataInterface() {propertyValues {Curve {keys [<at 0 36>,<at 0.12 36>,<at 0.13 0.1>,<at 0.75 0.1>,<at 0.76 36>,<at 1 36>]}}}
     CurveIndex=$Emitter.NormalizedLoopAge
    )
    Use Spawn Probability=false
   )
  ]
  particleSpawnScript [
   &InitializeParticle(
    Color=<rgba 1 1 1 0.45>
    Lifetime Max=1.2
    Lifetime Min=0.8
    Lifetime Mode=::NewEnumerator1
    Position=<10 18 -16>
    Position Mode=::NewEnumerator1
    Sprite Rotation Mode=::NewEnumerator2
    Sprite Size Max=<10 15>
    Sprite Size Min=<8 10>
    Sprite Size Mode=::NewEnumerator2
    UsePositionOffset=false
    Write Material Random=true
   )
   &OffsetPosition(Position Offset=<7.5 2.5 -4.5>)
   &AddVelocity(Velocity=&RandomRangeVector(Minimum=<0 35 0>, Maximum=<0 40 5>))
   &AddRotationalVelocity(Rotation Rate=&RandomRangeVector(Minimum=<-0.02 -0.02 -0.02>, Maximum=<0.02 0.02 0.02>))
   &DynamicMaterialParameters(
    Index 0 Param 1=0.05
    Index 0 Param 2=0.5
    Index 0 Param 3=2
    Module.Index0Float0Write=true
    Module.Index0Float1Write=true
    Module.Index0Float2Write=true
    Module.Index0Float3Write=true
   )
  ]
  particleUpdateScript [
   &ParticleState()
   &ScaleColor(
    Scale Alpha=&FloatFromCurve(FloatCurve=DataInterface() {propertyValues {Curve {keys [<at 0 0>,<at 0.058 0.15>,<at 0.75 0.5>,<at 1 0>]}}})
    ScaleA=true
    ScaleRGB=true
   )
   &ScaleVelocity(Velocity Scale=&VectorFromCurve(VectorCurve=DataInterface() {propertyValues {
       XCurve {keys [<at 0 0.2 RCIM_Cubic>,<at 0.15 0 RCIM_Cubic>,<at 0.8 0.2>,<at 1 0.2>]}
       YCurve {keys [
         <at 0.0038 0 RCIM_Cubic>
         {arriveTangent=0.7572, interpMode=RCIM_Cubic, leaveTangent=0.7572, time=0.1767, value=0.8}
         <at 0.4 0.3>
         <at 0.8 0.15>
         <at 1 0.15>
        ]}
       ZCurve {keys [<at 0 0 RCIM_Cubic>, {arriveTangent=0.25, interpMode=RCIM_Cubic, leaveTangent=0.25, time=0.15, value=0}, <at 0.8 0.2>, <at 1 0.2>]}
      }}))
   &SolveForcesAndVelocity(Clamp Velocity=false, Limit Acceleration=false)
   &GenerateLocationEvent(Delay Before Sending Events=0.45, Event Send Rate=2, Use Event Probability=false, UseEventDelay=true)
   &ScaleSpriteSize(Uniform Curve Sprite Scale=DataInterface() {propertyValues {Curve {keys [<at 0 0.05>, {tangentMode=RCTM_User, time=0.25, value=0.65}, <at 0.45 0.25>, <at 0.7079 0.0795>, <at 1 0>]}}})
  ]
  rendererClasses [&NiagaraSpriteRendererProperties]
  renderers [{rendererClass=&NiagaraSpriteRendererProperties}]
 }
 SteamTop {
  emitterUpdateScript [
   &EmitterState()
   &SpawnBurst_Instantaneous(Spawn Count=40, Spawn Time=0.1, Use Spawn Probability=false, UseLoopCountLimit=false) {enabled=false}
   &SpawnRate(
    SpawnRate=&FloatFromCurve(
     FloatCurve=DataInterface() {propertyValues {Curve {keys [<at 0 36>,<at 0.12 36>,<at 0.13 0.1>,<at 0.75 0.1>,<at 0.76 36>,<at 1 36>]}}}
     CurveIndex=$Emitter.NormalizedLoopAge
    )
    Use Spawn Probability=false
   )
  ]
  particleSpawnScript [
   &InitializeParticle(
    Color=<rgba 1 1 1 0.45>
    Lifetime Max=1.15
    Lifetime Min=0.8
    Lifetime Mode=::NewEnumerator1
    Position=<18 20 20>
    Position Mode=::NewEnumerator1
    Sprite Rotation Mode=::NewEnumerator2
    Sprite Size Max=<20 12>
    Sprite Size Min=<10.6 6.4>
    Sprite Size Mode=::NewEnumerator2
    UsePositionOffset=false
    Write Material Random=true
   )
   &AddVelocity(Velocity=&RandomRangeVector(Minimum=<0 22 0>, Maximum=<0 30 0>))
   &AddRotationalVelocity(Rotation Rate=&RandomRangeVector(Minimum=<-0.02 -0.02 -0.02>, Maximum=<0.02 0.02 0.02>)) {enabled=false}
   &DynamicMaterialParameters(
    Index 0 Param 1=1
    Index 0 Param 2=0.5
    Index 0 Param 3=2
    Module.Index0Float0Write=true
    Module.Index0Float1Write=true
    Module.Index0Float2Write=true
    Module.Index0Float3Write=true
   )
  ]
  particleUpdateScript [
   &ParticleState()
   &ScaleColor(
    Scale Alpha=&FloatFromCurve(FloatCurve=DataInterface() {propertyValues {Curve {keys [<at 0 0>,<at 0.058 0.25>,<at 0.75 0.4>,<at 1 0>]}}})
    ScaleA=true
    ScaleRGB=true
   )
   &SolveForcesAndVelocity(Clamp Velocity=false, Limit Acceleration=false)
   &GenerateLocationEvent(Event Type:ENiagara_LocEventType=::NewEnumerator1, Unit Spacing=25, Use Event Probability=false, UseEventDelay=false)
   &ScaleSpriteSize(Uniform Curve Sprite Scale=DataInterface() {propertyValues {Curve {keys [<at 0.2244 0.3251>, {tangentMode=RCTM_User, time=0.3, value=0.5}, <at 0.7 0.4>, <at 0.9872 0.3282>]}}})
  ]
  rendererClasses [&NiagaraSpriteRendererProperties]
  renderers [{rendererClass=&NiagaraSpriteRendererProperties}]
 }
}
summary {
 systemName=NS_NeonPunk
 emitters [
  {emitterName=SteamTop, rendererClasses [&NiagaraSpriteRendererProperties]}
  {emitterName=CloudTop, rendererClasses [&NiagaraSpriteRendererProperties]}
  {emitterName=SteamBottomRight, rendererClasses [&NiagaraSpriteRendererProperties]}
  {emitterName=CloudBottomRight, rendererClasses [&NiagaraSpriteRendererProperties]}
  {emitterName=SteamBottomLeft, rendererClasses [&NiagaraSpriteRendererProperties]}
  {emitterName=CloudBottomLeft, rendererClasses [&NiagaraSpriteRendererProperties]}
 ]
 userVariables []
}
